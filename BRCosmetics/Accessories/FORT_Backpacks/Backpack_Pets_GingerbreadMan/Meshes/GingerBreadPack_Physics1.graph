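# Physics asset for the Gingerbread Man pet backpack cosmetic, providing box bodies for its dangling chain bones and root.
# Top and bottom seven-link chains use tight swing and twist limits between links, with both ends tied to root to form suspended spans.
# The looser five-link side chain hangs from root and ends in a larger box.
# Bottom-end root joints rotate freely; the top's first link has both limited and free root constraints.
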PhysicsAsset /BRCosmetics/Accessories/FORT_Backpacks/Backpack_Pets_GingerbreadMan/Meshes/GingerBreadPack_Physics1
bodies {
 dyn_chain_btm [{shapeType=Box, extentX=7.5406, extentY=2.0129, extentZ=3.9365, center=<0.5 0 0>}]
 dyn_chain_btm1 [{shapeType=Box, extentX=7.5406, extentY=2.0129, extentZ=3.9365, center=<0.5 0 0>}]
 dyn_chain_btm2 [{shapeType=Box, extentX=7.5406, extentY=2.0129, extentZ=3.9365, center=<0.5 0 0>}]
 dyn_chain_btm3 [{shapeType=Box, extentX=7.5406, extentY=2.0129, extentZ=3.9365, center=<0.5 0 0>}]
 dyn_chain_btm4 [{shapeType=Box, extentX=7.5406, extentY=2.0129, extentZ=3.9365, center=<0.5 0 0>}]
 dyn_chain_btm5 [{shapeType=Box, extentX=7.5406, extentY=2.0129, extentZ=3.9365, center=<0.5 0 0>}]
 dyn_chain_btm6 [{shapeType=Box, extentX=7.5406, extentY=2.0129, extentZ=3.9365, center=<0.5 0 0>}]
 dyn_chain_side [{shapeType=Box, extentX=2.6074, extentY=2.6074, extentZ=2.5002, center=<1 0 0>}]
 dyn_chain_side1 [{shapeType=Box, extentX=2.6074, extentY=2.6074, extentZ=2.5002, center=<1 0 0>}]
 dyn_chain_side2 [{shapeType=Box, extentX=2.6074, extentY=2.6074, extentZ=2.5002, center=<1 0 0>}]
 dyn_chain_side3 [{shapeType=Box, extentX=2.6074, extentY=2.6074, extentZ=2.5002, center=<1 0 0>}]
 dyn_chain_side4 [{shapeType=Box, extentX=6.3715, extentY=2.6074, extentZ=6.622, center=<3.7434 0 0>}]
 dyn_chain_top [{shapeType=Box, extentX=7.5406, extentY=2.0129, extentZ=3.9365, center=<0.5 0 0>}]
 dyn_chain_top1 [{shapeType=Box, extentX=7.5406, extentY=2.0129, extentZ=3.9365, center=<0.5 0 0>}]
 dyn_chain_top2 [{shapeType=Box, extentX=7.5406, extentY=2.0129, extentZ=3.9365, center=<0.5 0 0>}]
 dyn_chain_top3 [{shapeType=Box, extentX=7.5406, extentY=2.0129, extentZ=3.9365, center=<0.5 0 0>}]
 dyn_chain_top4 [{shapeType=Box, extentX=7.5406, extentY=2.0129, extentZ=3.9365, center=<0.5 0 0>}]
 dyn_chain_top5 [{shapeType=Box, extentX=7.5406, extentY=2.0129, extentZ=3.9365, center=<0.5 0 0>}]
 dyn_chain_top6 [{shapeType=Box, extentX=7.5406, extentY=2.0129, extentZ=3.9365, center=<0.5 0 0>}]
 root [{shapeType=Box, extentX=1, extentY=1, extentZ=1, center=<0.5 0 0>}]
}
constraints [
 {bone1Name=dyn_chain_top, bone2Name=root, swing1LimitDegrees=45, swing2LimitDegrees=45, twistLimitDegrees=15}
 {bone1Name=dyn_chain_top1, bone2Name=dyn_chain_top, swing1LimitDegrees=10, swing2LimitDegrees=10, twistLimitDegrees=5}
 {bone1Name=dyn_chain_top2, bone2Name=dyn_chain_top1, swing1LimitDegrees=10, swing2LimitDegrees=10, twistLimitDegrees=5}
 {bone1Name=dyn_chain_top3, bone2Name=dyn_chain_top2, swing1LimitDegrees=10, swing2LimitDegrees=10, twistLimitDegrees=5}
 {bone1Name=dyn_chain_top4, bone2Name=dyn_chain_top3, swing1LimitDegrees=10, swing2LimitDegrees=10, twistLimitDegrees=5}
 {bone1Name=dyn_chain_top5, bone2Name=dyn_chain_top4, swing1LimitDegrees=10, swing2LimitDegrees=10, twistLimitDegrees=5}
 {bone1Name=dyn_chain_top6, bone2Name=dyn_chain_top5, swing1LimitDegrees=10, swing2LimitDegrees=10, twistLimitDegrees=5}
 {bone1Name=dyn_chain_btm1, bone2Name=dyn_chain_btm, swing1LimitDegrees=10, swing2LimitDegrees=10, twistLimitDegrees=5}
 {bone1Name=dyn_chain_btm2, bone2Name=dyn_chain_btm1, swing1LimitDegrees=10, swing2LimitDegrees=10, twistLimitDegrees=5}
 {bone1Name=dyn_chain_btm3, bone2Name=dyn_chain_btm2, swing1LimitDegrees=10, swing2LimitDegrees=10, twistLimitDegrees=5}
 {bone1Name=dyn_chain_btm4, bone2Name=dyn_chain_btm3, swing1LimitDegrees=10, swing2LimitDegrees=10, twistLimitDegrees=5}
 {bone1Name=dyn_chain_btm5, bone2Name=dyn_chain_btm4, swing1LimitDegrees=10, swing2LimitDegrees=10, twistLimitDegrees=5}
 {bone1Name=dyn_chain_btm6, bone2Name=dyn_chain_btm5, swing1LimitDegrees=10, swing2LimitDegrees=10, twistLimitDegrees=5}
 {bone1Name=dyn_chain_side, bone2Name=root, swing1LimitDegrees=45, swing2LimitDegrees=45, twistLimitDegrees=15}
 {bone1Name=dyn_chain_side1, bone2Name=dyn_chain_side, swing1LimitDegrees=45, swing2LimitDegrees=45, twistLimitDegrees=15}
 {bone1Name=dyn_chain_side2, bone2Name=dyn_chain_side1, swing1LimitDegrees=45, swing2LimitDegrees=45, twistLimitDegrees=15}
 {bone1Name=dyn_chain_side3, bone2Name=dyn_chain_side2, swing1LimitDegrees=45, swing2LimitDegrees=45, twistLimitDegrees=15}
 {bone1Name=dyn_chain_side4, bone2Name=dyn_chain_side3, swing1LimitDegrees=45, swing2LimitDegrees=45, twistLimitDegrees=15}
 {
  bone1Name=root
  bone2Name=dyn_chain_btm6
  swing1LimitDegrees=45
  swing1Motion=Free
  swing2LimitDegrees=45
  swing2Motion=Free
  twistLimitDegrees=45
  twistMotion=Free
 }
 {bone1Name=root, bone2Name=dyn_chain_top6, swing1LimitDegrees=45, swing2LimitDegrees=45, twistLimitDegrees=15}
 {
  bone1Name=root
  bone2Name=dyn_chain_btm
  swing1LimitDegrees=45
  swing1Motion=Free
  swing2LimitDegrees=45
  swing2Motion=Free
  twistLimitDegrees=45
  twistMotion=Free
 }
 {
  bone1Name=root
  bone2Name=dyn_chain_top
  swing1LimitDegrees=45
  swing1Motion=Free
  swing2LimitDegrees=45
  swing2Motion=Free
  twistLimitDegrees=45
  twistMotion=Free
 }
]
